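# Translucent UI material that draws the grapple weapon reticle from brackets and trapezoids stored in red and green channels of an SDF texture.
# Both shapes scale around center with BracketsScale, the trapezoids slightly offset in UV and scale.
# Distance-field functions build fills, shadows and a trapezoid glow.
# Opacity parameters weight the summed layers, while Color2 tints a sum blended toward extra glow by TrapezoidGlowOpacity.
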Material /GrappleWeapon/UI/Reticle/Material/M_UI_GrappleWeaponReticle
graph {
 properties {blendMode=BLEND_Translucent, materialDomain=MD_UI, pixelDepthOffsetMode=PDOM_Legacy, usedShadingModels=Default Lit}
 nodes [
  n1 = NamedRerouteUsage()
  n2 = ScalarParameter() {defaultValue=1, group=Opacity, parameterName=BracketsOpacity}
  n3 = NamedRerouteUsage()
  n4 = ScalarParameter() {defaultValue=1, group=Opacity, parameterName=TrapezoidOpacity}
  Add(
   A=Add(A=Multiply(A=@n1, B=@n2), B=Multiply(A=@n3, B=@n4))
   B=Multiply(A=NamedRerouteUsage(), B=ScalarParameter() {group=Opacity, parameterName=TrapezoidGlowOpacity})
  )
  n5 = Constant()
  n6 = Constant() {r=0.3}
  n7 = AppendVector(A=Constant(), B=Constant() {r=0.01})
  n8 = NamedRerouteUsage()
  n9 = ScalarParameter() {defaultValue=1, group=Scale, parameterName=BracketsScale}
  n10 = TextureSample(UVs=MaterialFunctionCall(UVs=Add(A=@n7, B=@n8), Texture Scale=Add(A=@n9) {constB=0.04}) {materialFunction=&ScaleUVsByCenter}.UVs) {samplerSource=SSM_Clamp_WorldGroupSettings, texture=&/T_UI_GrappleWeapon_Reticle_SDF}
  n11 = NamedRerouteDeclaration(None=MaterialFunctionCall!(RGBA Channels=@n10.G, Glow Min=Constant() {r=0.1}, Glow Max=Constant() {r=0.15}, Outline Thickness) {materialFunction=/FortUILibrary/MaterialFunctions/Basics/MF_UI_ChannelPackedDistanceField}.Fill) {name=Trapezoids, nodeColor=<rgba 0 0 0 1>}
  NamedRerouteDeclaration(None=LinearInterpolate(
    A=MaterialFunctionCall(RGBA Channels=@n10.G, Glow Min=@n5, Glow Max=@n6) {materialFunction=&MF_UI_ChannelPackedDistanceField}.Fill
    B=@n11
    Alpha=@n11
   )) {name=TrapezoidShadow, nodeColor=<rgba 0 0 0 1>}
  NamedRerouteDeclaration(None=TextureCoordinate()) {name=UV, nodeColor=<rgba 0 0 0 1>}
  n12 = Constant()
  n13 = Constant() {r=0.15}
  n14 = TextureSample(UVs=MaterialFunctionCall(UVs=NamedRerouteUsage(), Texture Scale=ScalarParameter() {defaultValue=1, group=Scale, parameterName=BracketsScale}) {materialFunction=&ScaleUVsByCenter}.UVs) {samplerSource=SSM_Clamp_WorldGroupSettings, texture=&T_UI_GrappleWeapon_Reticle_SDF}
  n15 = NamedRerouteDeclaration(None=MaterialFunctionCall(RGBA Channels=@n14.R, Glow Min=Constant() {r=0.1}, Glow Max=Constant() {r=0.15}) {materialFunction=&MF_UI_ChannelPackedDistanceField}.Fill) {name=Brackets, nodeColor=<rgba 0 0 0 1>}
  NamedRerouteDeclaration(None=LinearInterpolate(
    A=MaterialFunctionCall(RGBA Channels=@n14.R, Glow Min=@n12, Glow Max=@n13) {materialFunction=&MF_UI_ChannelPackedDistanceField}.Fill
    B=@n15
    Alpha=@n15
   )) {name=BracketsShadow, nodeColor=<rgba 0 0 0 1>}
  n16 = Constant()
  n17 = Constant() {r=0.1}
  NamedRerouteDeclaration(None=LinearInterpolate(
    A=MaterialFunctionCall(RGBA Channels=@n10.G, Glow Min=@n16, Glow Max=@n17) {materialFunction=&MF_UI_ChannelPackedDistanceField}.Fill
    B=@n11
    Alpha=@n11
   )) {name=TrapezoidGlow, nodeColor=<rgba 0 0 0 1>}
  n18 = Add(A=NamedRerouteUsage(), B=NamedRerouteUsage())
  n19 = Add(A=@n18, B=NamedRerouteUsage())
  n20 = ScalarParameter() {group=Opacity, parameterName=TrapezoidGlowOpacity}
  Saturate(None=Multiply(A=VectorParameter() {group=Colors, parameterName=Color2, defaultValue=<rgba 1 1 1 1>}.RGB, B=LinearInterpolate(A=@n18, B=@n19, Alpha=@n20)))
 ]
}
