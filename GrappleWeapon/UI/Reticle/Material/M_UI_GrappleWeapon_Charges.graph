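# Translucent UI material for the grapple weapon reticle, drawing available charges and a fill progress as a segmented arc.
# A circle outline, radial segments and rotated, offset UVs shape the arc, cropped to the visible segment count.
# Smoothed angle-gradient masks layer background, fill and charge colors, clamping the charge ratio to tolerate infinite charges.
# Scalar parameters drive first/final charge highlights, first-charge glow and final-charge opacity, alongside a soft shadow.
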Material /GrappleWeapon/UI/Reticle/Material/M_UI_GrappleWeapon_Charges
graph {
 properties {blendMode=BLEND_Translucent, materialDomain=MD_UI, pixelDepthOffsetMode=PDOM_Legacy, usedShadingModels=Default Lit}
 nodes [
  n1 = VectorParameter() {group=1- Charges, parameterName=Highlight Charge Color, defaultValue=<rgba 1 1 0 1>}
  n2 = NamedRerouteUsage()
  n3 = VectorParameter() {group=4 - Background, parameterName=BG Color, sortPriority=1, defaultValue=<rgba 0.0469 0.0197 0.0237 1>}
  n4 = NamedRerouteUsage()
  n5 = LinearInterpolate(
   A=LinearInterpolate(
    A=Multiply(A=@n3.RGB, B=@n4)
    B=VectorParameter() {group=2 - Fill, parameterName=Fill Color, sortPriority=2, defaultValue=<rgba 1 0 0.0636 1>}.RGB
    Alpha=NamedRerouteUsage()
   ) {constAlpha=0}
   B=VectorParameter() {group=1- Charges, parameterName=Charge Color, defaultValue=<rgba 1 1 1 1>}.RGB
   Alpha=NamedRerouteUsage()
  ) {constAlpha=1}
  n6 = LinearInterpolate(
   A=@n5
   B=LinearInterpolate(
    A=@n5
    B=VectorParameter() {group=1- Charges, parameterName=Highlight Charge Color, defaultValue=<rgba 1 1 0 1>}.RGB
    Alpha=NamedRerouteUsage()
   ) {constAlpha=1}
   Alpha=ScalarParameter() {group=1- Charges, parameterName=Final Charge Highlight Anim, sliderMax=1}
  )
  n7 = VectorParameter() {group=4 - Background, parameterName=BG Color, sortPriority=1, defaultValue=<rgba 0.0469 0.0197 0.0237 1>}
  n8 = LinearInterpolate(
   A=@n6
   B=LinearInterpolate(A=@n7.RGB, B=@n6, Alpha=ScalarParameter() {defaultValue=1, group=1- Charges, parameterName=Final Charge Opacity, sliderMax=1})
   Alpha=NamedRerouteUsage()
  )
  n9 = Multiply(
   A=LinearInterpolate(
    A=@n8
    B=LinearInterpolate(A=@n8, B=@n1.RGB, Alpha=@n2) {constAlpha=1}
    Alpha=ScalarParameter() {group=1- Charges, parameterName=First Charge Highlight Anim, sliderMax=1}
   )
   B=NamedRerouteUsage()
  )
  LinearInterpolate(
   A=@n9
   B=LinearInterpolate(
    A=@n9
    B=VectorParameter() {group=1- Charges, parameterName=Glow Charge Color, defaultValue=<rgba 0.2302 0.7934 1 1>}.RGB
    Alpha=NamedRerouteUsage()
   )
   Alpha=ScalarParameter() {parameterName=First Charge Glow Anim}
  ) {constAlpha=0}
  NamedRerouteDeclaration(None=Divide(B=ScalarParameter() {defaultValue=40, group=3 - Circle Settings, parameterName=Segments, sortPriority=5})) {name=SegmentRotation, nodeColor=<rgba 0 0 0 1>}
  n10 = NamedRerouteUsage()
  n11 = Constant() {r=1.5}
  n12 = Constant() {r=0.001}
  n13 = Constant()
  n14 = Constant() {r=0.3333}
  NamedRerouteDeclaration(None=SmoothStep(
    Min=ScalarParameter() {defaultValue=0.9, parameterName=Glow Width}
    Value=MaterialFunctionCall(UVs=@n10, Size=@n11, OutlineThickness=@n12, GlowMin=@n13, GlowMax=@n14) {materialFunction=&MF_UI_Circle}.Outline
   )) {name=Glow Radial Mask, nodeColor=<rgba 0.7344 0 1 1>}
  n15 = MaterialFunctionCall!(UVs=NamedRerouteUsage(), Rotation=ScalarParameter() {defaultValue=0.025, group=1- Charges, parameterName=Charge Rotation, sliderMax=1}) {materialFunction=/FortUILibrary/MaterialFunctions/Gradients/MF_UI_AngleGradient}
  n16 = Divide(
   A=ScalarParameter() {defaultValue=5, group=1- Charges, parameterName=Charges Available, sortPriority=0}
   B=ScalarParameter() {defaultValue=6, group=3 - Circle Settings, parameterName=Visible Segment Count, sortPriority=2}
  )
  n17 = Subtract(
   A=@n15.Result
   B=LinearInterpolate(B=NamedRerouteUsage(), Alpha=Clamp(None=@n16) {desc="Prevent issues from\r\nhaving infinite charges\r\n(arbitrarily large number)"})
  ) {constB=0.5}
  n18 = ScalarParameter() {defaultValue=0.0053, group=2 - Fill, parameterName=FillSmoothness}
  NamedRerouteDeclaration(None=OneMinus(None=MaterialFunctionCall!(In=@n17, Ramp Width=@n18) {materialFunction=/Engine/Functions/Engine_MaterialFunctions01/Math/SmoothCeil}.Result)) {name=Charge Mask, nodeColor=<rgba 0 0 0 1>}
  n19 = Subtract(A=MaterialFunctionCall() {materialFunction=&GetUserInterfaceUV}.Normalized UV, B=NamedRerouteUsage())
  n20 = ScalarParameter() {defaultValue=40, group=3 - Circle Settings, parameterName=Segments, sortPriority=5}
  n21 = ScalarParameter() {defaultValue=0.5, group=3 - Circle Settings, parameterName=Segment Spacing, sliderMax=1, sortPriority=6}
  NamedRerouteDeclaration(None=MaterialFunctionCall(
    Alpha=MaterialFunctionCall!(UV=@n19, Segment Count=@n20, Segment Spacing=@n21, Fixed Spacing) {materialFunction=/FortUILibrary/MaterialFunctions/Basics/MF_UI_RadialSegments}.Segmented Triangles Gradient
    Min=ScalarParameter() {defaultValue=0.17, group=3 - Circle Settings, parameterName=Segments Min Sharpness}
    Max=ScalarParameter() {defaultValue=0.65, group=3 - Circle Settings, parameterName=Segments Max Sharpness}
   ) {materialFunction=&SmoothStep}.Result) {name=Segments, nodeColor=<rgba 0 0 0 1>}
  n22 = NamedRerouteUsage()
  n23 = ScalarParameter() {defaultValue=0.025, group=1- Charges, parameterName=Charge Rotation, sliderMax=1}
  n24 = OneMinus(None=ScalarParameter() {defaultValue=0.5, group=3 - Circle Settings, parameterName=Segment Spacing, sliderMax=1, sortPriority=6})
  n25 = ScalarParameter() {defaultValue=0.65, group=3 - Circle Settings, parameterName=Segments Max Sharpness}
  n26 = OneMinus(None=@n25)
  n27 = Multiply(A=Subtract(A=@n25, B=ScalarParameter() {defaultValue=0.17, group=3 - Circle Settings, parameterName=Segments Min Sharpness})) {constB=0.15}
  n28 = Multiply(A=Add(A=Multiply(A=@n24, B=@n26), B=Multiply(A=@n24, B=@n27))) {constB=0.5}
  n29 = LinearInterpolate(
   A=@n28
   B=OneMinus(None=@n28)
   Alpha=ScalarParameter() {defaultValue=0.5318, group=2 - Fill, parameterName=Fill Progress, sliderMax=1, sortPriority=1}
  )
  n30 = Subtract(
   A=OneMinus(None=MaterialFunctionCall(UVs=@n22, Rotation=@n23) {materialFunction=&MF_UI_AngleGradient}.Result)
   B=LinearInterpolate(B=NamedRerouteUsage(), Alpha=Add(A=@n29) {constB=0.025})
  )
  n31 = ScalarParameter() {defaultValue=0.0053, group=2 - Fill, parameterName=FillSmoothness}
  NamedRerouteDeclaration(None=OneMinus(None=MaterialFunctionCall(In=@n30, Ramp Width=@n31) {materialFunction=&SmoothCeil}.Result)) {name=Fill Mask, nodeColor=<rgba 0 0 0 1>}
  NamedRerouteDeclaration(None=Multiply(A=NamedRerouteUsage(), B=NamedRerouteUsage())) {name=Charges Mask, nodeColor=<rgba 0 0 0 1>}
  n32 = MaterialFunctionCall() {materialFunction=&GetUserInterfaceUV}
  n33 = NamedRerouteDeclaration(None=AppendVector(
    A=ScalarParameter() {defaultValue=-0.596, group=3 - Circle Settings, parameterName=Offset X}
    B=ScalarParameter() {defaultValue=-0.188, group=3 - Circle Settings, parameterName=Offset Y}
   )) {name=Offset, nodeColor=<rgba 0 0 0 1>}
  NamedRerouteDeclaration(None=MaterialFunctionCall(
    UVs=Subtract(A=@n32.Normalized UV, B=@n33)
    "Rotation Angle (0-1)"=ScalarParameter() {defaultValue=0.6, group=3 - Circle Settings, parameterName=Rotation}
   ) {materialFunction=&CustomRotator}.Rotated Values) {name=UVs, nodeColor=<rgba 0 0 0 1>}
  n34 = NamedRerouteUsage()
  n35 = OneMinus(None=MaterialFunctionCall(
    In=Subtract(A=@n15.Result, B=@n34) {constB=0.5}
    Ramp Width=ScalarParameter() {defaultValue=0.0053, group=2 - Fill, parameterName=FillSmoothness}
   ) {materialFunction=&SmoothCeil}.Result)
  NamedRerouteDeclaration(None=@n35) {name=FinalChargeMask, nodeColor=<rgba 0 0 0 1>}
  n36 = NamedRerouteUsage()
  n37 = NamedRerouteUsage()
  n38 = NamedRerouteDeclaration(None=Multiply(A=OneMinus(None=@n37), B=ScalarParameter() {defaultValue=0.25, group=4 - Background, parameterName=BG Opacity, sliderMax=1, sortPriority=2})) {name=BG Color Mask, nodeColor=<rgba 0 0 0 1>}
  NamedRerouteDeclaration(None=Saturate(None=Multiply(A=@n36, B=@n38))) {name=BG Opacity, nodeColor=<rgba 0 0 0 1>}
  n39 = NamedRerouteUsage()
  n40 = NamedRerouteUsage()
  n41 = NamedRerouteDeclaration(None=Multiply(A=@n40, B=ScalarParameter() {defaultValue=1, group=2 - Fill, parameterName=Fill Opacity, sliderMax=1, sortPriority=3})) {name=Fill Color Mask, nodeColor=<rgba 0 0 0 1>}
  NamedRerouteDeclaration(None=Saturate(None=Multiply(A=@n39, B=@n41))) {name=Fill Opacity, nodeColor=<rgba 0 0 0 1>}
  NamedRerouteDeclaration(None=@n35) {name=FirstChargeMask, nodeColor=<rgba 0 0 0 1>}
  n42 = MaterialFunctionCall(
   UVs=@n10
   Size=ScalarParameter() {defaultValue=1.5, group=3 - Circle Settings, parameterName=Outer Size, sortPriority=0}
   OutlineThickness=ScalarParameter() {defaultValue=0.03, group=3 - Circle Settings, parameterName=Outline Thickness, sortPriority=1}
   GlowMin=ScalarParameter() {group=3 - Circle Settings, parameterName=Edge Softness Min, sortPriority=2}
   GlowMax=ScalarParameter() {defaultValue=0.1, group=3 - Circle Settings, parameterName=Edge Softness Max, sortPriority=3}
  ) {materialFunction=&MF_UI_Circle}
  n43 = Multiply(A=@n42.Outline, B=ScalarParameter() {defaultValue=0.8, group=3 - Circle Settings, parameterName=ShadowOpacity})
  n44 = NamedRerouteDeclaration(None=SmoothStep(Value=@n42.Outline) {constMax=0.5, constMin=0.3}) {name=Circle Mask, nodeColor=<rgba 0 0 0 1>}
  NamedRerouteDeclaration(None=Saturate(None=Subtract(A=@n43, B=@n44))) {name=shadow, nodeColor=<rgba 0 1 0.6016 1>}
  n45 = MaterialFunctionCall(
   UV=Subtract(A=MaterialFunctionCall() {materialFunction=&GetUserInterfaceUV}.Normalized UV, B=NamedRerouteUsage())
   Segment Count=Constant() {r=40}
   Segment Spacing=Constant()
  ) {materialFunction=&MF_UI_RadialSegments}
  n46 = Constant() {r=0.5}
  n47 = Constant() {r=1}
  NamedRerouteDeclaration(None=OneMinus(None=MaterialFunctionCall(Alpha=@n45.Segmented Triangles Gradient, Min=@n46, Max=@n47) {materialFunction=&SmoothStep}.Result)) {name=SegmentGlow, nodeColor=<rgba 0 0 0 1>}
  n48 = NamedRerouteUsage()
  n49 = LinearInterpolate(
   A=ScalarParameter() {defaultValue=0.25, group=4 - Background, parameterName=BG Opacity, sliderMax=1, sortPriority=2}
   Alpha=ScalarParameter() {defaultValue=1, group=1- Charges, parameterName=Final Charge Opacity, sliderMax=1}
  )
  n50 = NamedRerouteUsage()
  n51 = Add(
   A=LinearInterpolate(A=NamedRerouteUsage(), B=@n48, Alpha=@n48)
   B=Multiply(A=NamedRerouteUsage(), B=LinearInterpolate(B=@n49, Alpha=@n50) {constA=1})
  )
  n52 = Multiply(A=NamedRerouteUsage(), B=NamedRerouteUsage())
  n53 = Add(A=@n51, B=Multiply(A=NamedRerouteUsage(), B=NamedRerouteDeclaration(None=@n52) {name=Final Charge Glow Mask, nodeColor=<rgba 0.3359 0 1 1>}))
  n54 = ScalarParameter() {parameterName=First Charge Glow Anim}
  n55 = Add(
   A=NamedRerouteUsage()
   B=Multiply(
    A=LinearInterpolate(A=@n51, B=@n53, Alpha=@n54) {constAlpha=1}
    B=ScalarParameter() {defaultValue=1, parameterName=Global Opacity, sliderMax=1}
   )
  )
  n56 = NamedRerouteDeclaration(None=Divide(
    A=ScalarParameter() {defaultValue=6, group=3 - Circle Settings, parameterName=Visible Segment Count, sortPriority=2}
    B=ScalarParameter() {defaultValue=40, group=3 - Circle Settings, parameterName=Segments, sortPriority=5}
   )) {name=Segment Mask, nodeColor=<rgba 0 0 0 1>}
  n57 = OneMinus(None=MaterialFunctionCall!(UVs=NamedRerouteUsage(), EdgeSoftness, Progress=OneMinus(None=@n56)) {materialFunction=/Game/UI/Foundation/Toasts/Materials/MF_RadialMask}.Mask)
  Reroute(None=Subtract(A=Multiply(A=@n55, B=@n57), B=NamedRerouteUsage()))
  Subtract(A=@n40)
 ]
}
