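# Translucent UI material that draws the grapple weapon's point-of-attachment reticle.
# Named reroutes build a hexagon from an SDF texture, a rotating segmented circle loader, and a procedural dot and six-vertex arrow, both with soft shadows.
# Chained lerps on the ChangetoAttack, AttachmentInSight and ChangetoZip scalars switch both color and opacity between center-shape states.
# Parameters expose colors, scales, opacities and the loader's speed and divisions.
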Material /GrappleWeapon/UI/Reticle/Material/M_UI_GrappleWeapon_PointOfAttachment
graph {
 properties {blendMode=BLEND_Translucent, materialDomain=MD_UI, pixelDepthOffsetMode=PDOM_Legacy, usedShadingModels=Default Lit}
 nodes [
  n1 = NamedRerouteUsage()
  n2 = VectorParameter() {group=Colors, parameterName=Color1, defaultValue=<rgba 0.0685 0.5089 1 1>}
  n3 = Multiply(A=NamedRerouteUsage(), B=@n2.RGB)
  n4 = Add(A=Multiply(A=NamedRerouteUsage(), B=VectorParameter() {parameterName=Color 2, defaultValue=<rgba 1 1 1 1>}.RGB), B=@n3)
  n5 = ScalarParameter() {group=CenterShape, parameterName=ChangetoAttack, sortPriority=1}
  LinearInterpolate(
   A=LinearInterpolate(
    A=LinearInterpolate(A=@n1, B=@n4, Alpha=@n5)
    B=@n3
    Alpha=ScalarParameter() {group=CenterShape, parameterName=AttachmentInSight, sortPriority=2}
   )
   B=Multiply(A=NamedRerouteUsage(), B=@n2.RGB)
   Alpha=ScalarParameter() {group=CenterShape, parameterName=ChangetoZip, sortPriority=3}
  ) {constAlpha=1}
  n6 = NamedRerouteUsage()
  n7 = LinearInterpolate(
   A=NamedRerouteUsage()
   B=Add(A=NamedRerouteUsage(), B=@n6)
   Alpha=ScalarParameter() {group=CenterShape, parameterName=ChangetoAttack, sortPriority=1}
  )
  n8 = ScalarParameter() {group=CenterShape, parameterName=AttachmentInSight, sortPriority=2}
  Multiply(
   A=LinearInterpolate(
    A=LinearInterpolate(A=@n7, B=@n6, Alpha=@n8)
    B=NamedRerouteUsage()
    Alpha=ScalarParameter() {group=CenterShape, parameterName=ChangetoZip, sortPriority=3}
   ) {constAlpha=1}
   B=ScalarParameter() {defaultValue=1, group=Opacity, parameterName=CenterOpacity, sortPriority=4}
  )
  n9 = Constant() {r=0.1}
  n10 = Constant() {r=0.2}
  n11 = TextureSample(UVs=MaterialFunctionCall(UVs=NamedRerouteUsage(), Texture Scale=ScalarParameter() {defaultValue=0.4, group=Scale, parameterName=ExagonScale}) {materialFunction=&ScaleUVsByCenter}.UVs) {samplerSource=SSM_Clamp_WorldGroupSettings, texture=&/T_UI_GrappleWeapon_Reticle_SDF}
  NamedRerouteDeclaration(None=Multiply(
    A=MaterialFunctionCall!(RGBA Channels=@n11.B, Glow Min=@n9, Glow Max=@n10, Outline Thickness) {materialFunction=/FortUILibrary/MaterialFunctions/Basics/MF_UI_ChannelPackedDistanceField}.Fill
    B=ScalarParameter() {defaultValue=1, group=Opacity, parameterName=ExagonOpacity}
   )) {name=Exagon, nodeColor=<rgba 0 0 0 1>}
  n12 = ScalarParameter() {defaultValue=2, group=Loader, parameterName=Divisions}
  n13 = NamedRerouteUsage()
  n14 = Multiply(A=Time(), B=Multiply(A=ScalarParameter() {defaultValue=60, group=Loader, parameterName=Speed}) {constB=-1})
  n15 = MaterialFunctionCall(Vector or UVs=Add(A=Rotator(Coordinate=@n13, Time=@n14) {speed=0.21}) {constB=-0.5}) {materialFunction=&VectorToRadialValue}
  n16 = LinearInterpolate(A=Constant() {r=-0.5}, B=Constant() {r=1}, Alpha=Frac(None=Multiply(A=@n12, B=@n15.Vector Converted to Angle) {constA=2}))
  n17 = NamedRerouteUsage()
  n18 = Constant() {r=0.35}
  n19 = Constant() {r=0.06}
  n20 = OneMinus(None=SmoothStep(Value=MaterialFunctionCall(UVs=@n17, Size=@n18, OutlineThickness=@n19) {materialFunction=&MF_UI_Circle}.SDFOutline) {constMax=0.02, constMin=0.01})
  NamedRerouteDeclaration(None=Saturate(None=Multiply(A=@n16, B=@n20))) {name=CircleLoader, nodeColor=<rgba 0 0 0 1>}
  n21 = MaterialFunctionCall(UVs=NamedRerouteUsage(), Size=ScalarParameter() {defaultValue=0.04, group=Scale, parameterName=DotScale}) {materialFunction=&MF_UI_Circle}
  n22 = SmoothStep(Value=@n21.SDF) {constMax=0.06, constMin=0.0001}
  n23 = NamedRerouteDeclaration(None=OneMinus(None=SmoothStep(Value=@n21.SDF) {constMax=0.02, constMin=0.01})) {name=Dot, nodeColor=<rgba 0 0 0 1>}
  NamedRerouteDeclaration(None=LinearInterpolate(A=OneMinus(None=@n22), B=@n23, Alpha=@n23)) {name=DotShadow, nodeColor=<rgba 0 0 0 1>}
  n24 = AppendVector(A=Constant(), B=Constant() {r=-0.015})
  n25 = NamedRerouteUsage()
  n26 = MaterialFunctionCall!(
   UVs=MaterialFunctionCall(UVs=Add(A=@n24, B=@n25), Texture Scale=Constant() {r=0.16}) {materialFunction=&ScaleUVsByCenter}.UVs
   vertex_0=Constant2Vector() {g=-0.4106, r=-1}
   vertex_1=Constant2Vector() {g=-0.7, r=-0.68}
   vertex_2=Constant2Vector() {r=-0.0016}
   vertex_4=Constant2Vector() {g=-0.4007, r=1}
   vertex_3=Constant2Vector() {g=-0.7, r=0.68}
   vertex_5=Constant2Vector() {g=0.6079, r=0.002}
   Outline Thickness
   Glow Min
   Glow Max
  ) {materialFunction=&/MF_UI_6VertexPolygon}
  n27 = Subtract(A=OneMinus(None=@n26.SDF), B=Constant() {r=0.7}) {constB=0.85}
  n28 = NamedRerouteDeclaration(None=@n26.Fill) {name=Arrow, nodeColor=<rgba 0 0 0 1>}
  NamedRerouteDeclaration(None=LinearInterpolate(A=Saturate(None=@n27), B=@n28, Alpha=@n28)) {name=ArrowShadow, nodeColor=<rgba 0 0 0 1>}
  NamedRerouteDeclaration(None=TextureCoordinate()) {name=UV, nodeColor=<rgba 0 0 0 1>}
 ]
}
