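# Grapple weapon reticle widget that marks the point of attachment on screen and shows targeting states such as no target, locked on and zipping.
# UpdatePointOfAttachment projects the world location into a viewport-centred, scale-corrected render translation, or resets the transform.
# SetPointOfAttachmentState plays each state's animation, choosing lock-on variants from the previous state and reset flag.
# Animations drive the image material's parameters, with sequencer events controlling a looping lock-on pulse.
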WidgetBlueprint /GrappleWeapon/UI/Reticle/WBP_Reticle_GrappleWeapon_PointOfAttachment
animations [
 Anim_LockedOn {
  binding_count=1
  length_seconds=0.4
  object=&self:Anim_LockedOn
  sequence {
   bindings [Image_ReticlePointOfAttachment {
     id=#1
     possessed_class=Image
     tracks [{
       type=MovieSceneWidgetMaterialTrack
       sections [{
         type=MovieSceneParameterSection
         end_frame=8
         start_frame=0
         channels [
          ExagonScale {type=MovieSceneScriptingFloatChannel, keys [[0,0.3935],[3,0.6],[8,0.4]]}
          CenterOpacity {type=MovieSceneScriptingFloatChannel, keys [[0,1],[3,0],[5,1]]}
          DotScale {type=MovieSceneScriptingFloatChannel, keys [[0,0.03],[3,-0.02]]}
          AttachmentInSight {type=MovieSceneScriptingFloatChannel, keys [[0,0],[2,0],[3,1]]}
          ChangetoAttack {type=MovieSceneScriptingFloatChannel, keys [[0,0],[2,0],[3,1]]}
         ]
        }]
      }]
    }]
   tracks [{
     type=MovieSceneEventTrack
     sections [{
       type=MovieSceneEventTriggerSection
       channels [None {
         type=MovieSceneScriptingEventChannel
         keys [[0, "<Struct 'MovieSceneEvent' (0x46883ffb360) {}>"], [8, "<Struct 'MovieSceneEvent' (0x46883ffac80) {}>"]]
        }]
      }]
    }]
  }
 }
 Anim_Zipping {
  binding_count=2
  length_seconds=0.35
  object=&self:Anim_Zipping
  sequence {
   bindings [
    WBP_Reticle_GrappleWeapon_PointOfAttachment_C_269 {
     id=#2
     possessed_class=WBP_Reticle_GrappleWeapon_PointOfAttachment_C
     tracks [{
       type=MovieScene2DTransformTrack
       sections [{
         type=MovieScene2DTransformSection
         end_frame=7
         start_frame=0
         channels [
          Translation.X {type=MovieSceneScriptingFloatChannel, const [0,1]}
          Translation.Y {type=MovieSceneScriptingFloatChannel, const [0,1]}
         ]
        }]
      }]
    }
    Image_ReticlePointOfAttachment {
     id=#3
     possessed_class=Image
     tracks [{
       type=MovieSceneWidgetMaterialTrack
       sections [{
         type=MovieSceneParameterSection
         end_frame=6
         start_frame=0
         channels [
          ExagonScale {type=MovieSceneScriptingFloatChannel, const [0.32,1]}
          CenterOpacity {type=MovieSceneScriptingFloatChannel, keys [[0,1],[2.618,0],[3.2727,0],[6.5453,1]]}
          DotScale {type=MovieSceneScriptingFloatChannel, const [-0.02,1]}
          AttachmentInSight {type=MovieSceneScriptingFloatChannel, const [1,1]}
          ChangetoZip {type=MovieSceneScriptingFloatChannel, keys [[2.618,0],[3.2727,1]]}
          ChangetoAttack {type=MovieSceneScriptingFloatChannel, const [1,2]}
         ]
        }]
      }]
    }
   ]
   tracks [{
     type=MovieSceneEventTrack
     sections [{
       type=MovieSceneEventTriggerSection
       channels [None {type=MovieSceneScriptingEventChannel, const ["<Struct 'MovieSceneEvent' (0x46883ffb400) {}>", 1]}]
      }]
    }]
  }
 }
 Anim_NoTarget {
  binding_count=1
  length_seconds=0.3
  object=&self:Anim_NoTarget
  sequence {
   bindings [Image_ReticlePointOfAttachment {
     id=#4
     possessed_class=Image
     tracks [{
       type=MovieSceneWidgetMaterialTrack
       sections [{
         type=MovieSceneParameterSection
         end_frame=6
         start_frame=0
         channels [
          BracketsScale {type=MovieSceneScriptingFloatChannel, const [1,1]}
          ExagonScale {type=MovieSceneScriptingFloatChannel, const [0.3935,1]}
          CenterOpacity {type=MovieSceneScriptingFloatChannel, const [1,1]}
          DotScale {type=MovieSceneScriptingFloatChannel, const [0.03,2]}
          AttachmentInSight {type=MovieSceneScriptingFloatChannel, const [0,1]}
          BracketsOpacity {type=MovieSceneScriptingFloatChannel, const [0.25,1]}
          ChangetoZip {type=MovieSceneScriptingFloatChannel, const [0,1]}
          ChangetoAttack {type=MovieSceneScriptingFloatChannel, const [0,1]}
         ]
        }]
      }]
    }]
   tracks [{
     type=MovieSceneEventTrack
     sections [{
       type=MovieSceneEventTriggerSection
       channels [None {type=MovieSceneScriptingEventChannel, const ["<Struct 'MovieSceneEvent' (0x46883ffc800) {}>", 1]}]
      }]
    }]
  }
 }
 Anim_LockedOnLoop {
  binding_count=1
  length_seconds=0.8759
  object=&self:Anim_LockedOnLoop
  sequence {bindings [Image_ReticlePointOfAttachment {
     id=#5
     possessed_class=Image
     tracks [{
       type=MovieSceneWidgetMaterialTrack
       sections [{
         type=MovieSceneParameterSection
         end_frame=17
         start_frame=0
         channels [CenterOpacity {type=MovieSceneScriptingFloatChannel, keys [[0,1],[9,0.3515],[17.518,1]]}]
        }]
      }]
    }]}
 }
 Anim_LockedOn_Player {
  binding_count=1
  length_seconds=0.25
  object=&self:Anim_LockedOn_Player
  sequence {
   bindings [Image_ReticlePointOfAttachment {
     id=#6
     possessed_class=Image
     tracks [{
       type=MovieSceneWidgetMaterialTrack
       sections [{
         type=MovieSceneParameterSection
         end_frame=5
         start_frame=0
         channels [
          ExagonScale {type=MovieSceneScriptingFloatChannel, keys [[0,0.3935],[2.3723,0.7],[5,0.4]]}
          CenterOpacity {type=MovieSceneScriptingFloatChannel, keys [[0,1],[3,0],[4.25,1]]}
          DotScale {type=MovieSceneScriptingFloatChannel, keys [[0,0.03],[3,-0.02]]}
          AttachmentInSight {type=MovieSceneScriptingFloatChannel, const [0,2]}
          ChangetoAttack {type=MovieSceneScriptingFloatChannel, keys [[2.375,0],[3,1]]}
          ExagonOpacity {type=MovieSceneScriptingFloatChannel, const [1,2]}
         ]
        }]
      }]
    }]
   tracks [{
     type=MovieSceneEventTrack
     sections [{
       type=MovieSceneEventTriggerSection
       channels [None {type=MovieSceneScriptingEventChannel, const ["<Struct 'MovieSceneEvent' (0x46883ffc940) {}>", 1]}]
      }]
    }]
  }
 }
 Anim_StopZipping {
  binding_count=1
  length_seconds=0.3
  object=&self:Anim_StopZipping
  sequence {bindings [Image_ReticlePointOfAttachment {
     id=#3
     possessed_class=Image
     tracks [{
       type=MovieSceneWidgetMaterialTrack
       sections [{
         type=MovieSceneParameterSection
         end_frame=6
         start_frame=0
         channels [
          BracketsScale {type=MovieSceneScriptingFloatChannel, const [0.8,1]}
          ExagonScale {type=MovieSceneScriptingFloatChannel, const [0.4,1]}
          CenterOpacity {type=MovieSceneScriptingFloatChannel, keys [[0,1],[3,0],[4,0],[6,1]]}
          DotScale {type=MovieSceneScriptingFloatChannel, const [-0.02,1]}
          AttachmentInSight {type=MovieSceneScriptingFloatChannel, const [1,1]}
          ChangetoZip {type=MovieSceneScriptingFloatChannel, keys [[3,1],[4,0],[6,0]]}
          BracketsOpacity {type=MovieSceneScriptingFloatChannel, const [0.5,1]}
          ChangetoAttack {type=MovieSceneScriptingFloatChannel, const [1,1]}
         ]
        }]
      }]
    }]}
 }
 Anim_LockedOn_WithoutScale {
  binding_count=1
  length_seconds=0.25
  object=&self:Anim_LockedOn_WithoutScale
  sequence {
   bindings [Image_ReticlePointOfAttachment {
     id=#1
     possessed_class=Image
     tracks [{
       type=MovieSceneWidgetMaterialTrack
       sections [{
         type=MovieSceneParameterSection
         end_frame=4
         start_frame=0
         channels [
          CenterOpacity {type=MovieSceneScriptingFloatChannel, const [1,2]}
          ChangetoAttack {type=MovieSceneScriptingFloatChannel, const [1,1]}
          ChangetoZip {type=MovieSceneScriptingFloatChannel, const [0,1]}
          AttachmentInSight {type=MovieSceneScriptingFloatChannel, const [1,1]}
          ExagonScale {type=MovieSceneScriptingFloatChannel, const [0.4,1]}
         ]
        }]
      }]
    }]
   tracks [{
     type=MovieSceneEventTrack
     sections [{
       type=MovieSceneEventTriggerSection
       channels [None {
         type=MovieSceneScriptingEventChannel
         keys [[0, "<Struct 'MovieSceneEvent' (0x46883ffd0c0) {}>"], [5, "<Struct 'MovieSceneEvent' (0x46883ffd020) {}>"]]
        }]
      }]
    }]
  }
 }
 Anim_LockedOn_FromZipping {
  binding_count=1
  length_seconds=0.25
  object=&self:Anim_LockedOn_FromZipping
  sequence {
   bindings [Image_ReticlePointOfAttachment {
     id=#1
     possessed_class=Image
     tracks [{
       type=MovieSceneWidgetMaterialTrack
       sections [{
         type=MovieSceneParameterSection
         end_frame=5
         start_frame=0
         channels [
          ChangetoAttack {type=MovieSceneScriptingFloatChannel, const [1,1]}
          ChangetoZip {type=MovieSceneScriptingFloatChannel, keys [[0,1],[5,0]]}
          AttachmentInSight {type=MovieSceneScriptingFloatChannel, const [1,1]}
          CenterOpacity {type=MovieSceneScriptingFloatChannel, const [1,1]}
          ExagonScale {type=MovieSceneScriptingFloatChannel, const [0.4,1]}
         ]
        }]
      }]
    }]
   tracks [{
     type=MovieSceneEventTrack
     sections [{
       type=MovieSceneEventTriggerSection
       channels [None {type=MovieSceneScriptingEventChannel, const ["<Struct 'MovieSceneEvent' (0x46883ffd340) {}>", 1]}]
      }]
    }]
  }
 }
 Anim_Reset {
  binding_count=1
  length_seconds=0.0551
  object=&self:Anim_Reset
  sequence {bindings [Image_ReticlePointOfAttachment {
     id=#7
     possessed_class=Image
     tracks [{
       type=MovieSceneWidgetMaterialTrack
       sections [{
         type=MovieSceneParameterSection
         end_frame=1
         start_frame=0
         channels [
          ChangetoAttack {type=MovieSceneScriptingFloatChannel, const [0,1]}
          ChangetoZip {type=MovieSceneScriptingFloatChannel, const [0,1]}
          AttachmentInSight {type=MovieSceneScriptingFloatChannel, const [0,1]}
         ]
        }]
      }]
    }]}
 }
]
graphs [
 EventGraph {nodes [
   Event Destruct()
   > SetRenderTransform!(
    execute
    self:Widget Object Reference
    InTransform_Translation:Vector 2D Structure=(X=0,Y=0)
    InTransform_Scale:Vector 2D Structure=(X=1,Y=1)
    InTransform_Shear:Vector 2D Structure=(X=0,Y=0)
    InTransform_Angle:float=0
   ) -> (then)
   > StopAllAnimations(self:User Widget Object Reference)
  ]}
 GetLockedOnAnimation {nodes [
   n1 = GetLockedOnAnimation!() -> (then, NeedsToResetReticleState:bool=~) {is=title K2Node_FunctionEntry}
   > n2 = Branch(Condition=@n1)
   > Return Node!(execute, LockedOnAnimation:Widget Animation Object Reference=Get Anim_LockedOn() -> (Anim_LockedOn:Widget Animation Object Reference))
   n3 = Get Anim_LockedOn_WithoutScale() -> (Anim_LockedOn_WithoutScale:Widget Animation Object Reference)
   n4 = Get Anim_LockedOn_FromZipping() -> (Anim_LockedOn_FromZipping:Widget Animation Object Reference)
   Return Node(
    execute=@n2.else
    LockedOnAnimation=Select!(
     NewEnumerator0:Widget Animation Object Reference=Get Anim_LockedOn()
     NewEnumerator1:Widget Animation Object Reference
     NewEnumerator7:Widget Animation Object Reference=Get Anim_LockedOn_WithoutScale()
     NewEnumerator2:Widget Animation Object Reference
     NewEnumerator3:Widget Animation Object Reference
     NewEnumerator4:Widget Animation Object Reference=Select!(Option 0:Widget Animation Object Reference=@n4, Option 1:Widget Animation Object Reference=@n3, Index:bool=@n1) -> (ReturnValue)
     NewEnumerator5:Widget Animation Object Reference
     NewEnumerator6:Widget Animation Object Reference
     Index:GrappleKnifeReticleState Enum=Get PreviousReticleState() -> (PreviousReticleState:GrappleKnifeReticleState Enum=NewEnumerator0)
    ) -> (ReturnValue:Widget Animation Object Reference)
   )
  ]}
 Sequencer Events {nodes [
   StartLockedOnLoop() {is=title K2Node_CustomEvent}
   > PlayAnimation(InAnimation=Get Anim_LockedOnLoop() -> (Anim_LockedOnLoop:Widget Animation Object Reference), NumLoopsToPlay=0, bRestoreState=true)
   StopLockedOnLoop() {is=title K2Node_CustomEvent}
   > StopAnimation(InAnimation=Get Anim_LockedOnLoop())
  ]}
 SetPointOfAttachmentPosition {nodes [
   SetPointOfAttachmentPosition() {is=title K2Node_FunctionEntry}
   n1 = Get LatestWorldLocation() -> (LatestWorldLocation:Vector)
   > Set RenderTransform!(
    execute
    RenderTransform:Widget Transform Structure=Make Widget Transform!(
     Translation:Vector 2D Structure=WorldToScreen!(WorldLocation:Vector=@n1) -> (ScreenLocation:Vector 2D Structure)
     Scale:Vector 2D Structure=(X=1,Y=1)
     Shear:Vector 2D Structure=(X=0,Y=0)
     Angle:float=0
    ) -> (WidgetTransform:Widget Transform Structure) {is=title K2Node_MakeStruct}
    self:Widget Object Reference
   ) -> (then, Output_Get:Widget Transform Structure)
   > Branch(Condition="Not Equal (Boolean)"(A=Get bLatestTargetingLedge() -> (bLatestTargetingLedge:bool), B=Get TargetingLedge() -> (TargetingLedge:bool)) -> (ReturnValue=~))
   > Set bLatestTargetingLedge(bLatestTargetingLedge:bool=Get bTargetingLedge() -> (bTargetingLedge:bool)) -> (Output_Get:bool)
  ]}
 SetPointOfAttachmentState {nodes [
   n1 = SetPointOfAttachmentState!() -> (then, ReticleState:GrappleKnifeReticleState Enum=NewEnumerator0, NeedsToResetReticleState:bool=~) {is=title K2Node_FunctionEntry}
   > StopAllAnimations()
   n2 = Select!(
    NewEnumerator0:Widget Animation Object Reference=Get Anim_NoTarget() -> (Anim_NoTarget:Widget Animation Object Reference)
    NewEnumerator1:Widget Animation Object Reference=GetLockedOnAnimation!(NeedsToResetReticleState:bool=@n1.NeedsToResetReticleState) -> (LockedOnAnimation:Widget Animation Object Reference) {is=title K2Node_CallFunction}
    NewEnumerator7:Widget Animation Object Reference=Get Anim_LockedOn_Player() -> (Anim_LockedOn_Player:Widget Animation Object Reference)
    NewEnumerator2:Widget Animation Object Reference=Get Anim_Zipping() -> (Anim_Zipping:Widget Animation Object Reference)
    NewEnumerator3:Widget Animation Object Reference
    NewEnumerator4:Widget Animation Object Reference
    NewEnumerator5:Widget Animation Object Reference
    NewEnumerator6:Widget Animation Object Reference
    Index:GrappleKnifeReticleState Enum=@n1
   ) -> (ReturnValue)
   > n3 = Is Valid(InputObject=@n2)
   > n4 = PlayAnimation(InAnimation=@n2)
   Set PreviousReticleState(execute=[@n4.then, @n3.Is Not Valid], PreviousReticleState:GrappleKnifeReticleState Enum=@n1) -> (Output_Get:GrappleKnifeReticleState Enum=NewEnumerator0)
  ]}
 UpdatePointOfAttachment {nodes [
   n1 = UpdatePointOfAttachment!() -> (
    then
    NeedsToUpdatePosition:bool=~
    WorldLocation:Vector=~
    TargetingLedge:bool=~
    NeedsToUpdateReticleState:bool=~
    ReticleState:GrappleKnifeReticleState Enum=NewEnumerator0
    NeedsToResetReticleState:bool=~
   ) {is=title K2Node_FunctionEntry}
   > n2 = Branch(Condition=AND Boolean!(
     A=@n1
     B="Not Equal (Enum)"!(A:GrappleKnifeReticleState Enum=@n1.ReticleState, B:GrappleKnifeReticleState Enum=NewEnumerator0) -> (ReturnValue:bool)
     C:bool="Not Equal (Enum)"(A=@n1.ReticleState, B=NewEnumerator3)
    ) -> (ReturnValue))
   > Set TargetingLedge(TargetingLedge:bool=@n1.TargetingLedge) -> (Output_Get:bool)
   > Set LatestWorldLocation(LatestWorldLocation:Vector=@n1.WorldLocation) -> (Output_Get:Vector)
   > n3 = SetPointOfAttachmentPosition() {is=title K2Node_CallFunction}
   Branch(execute=[@n3.then, @n4.then], Condition=@n1.NeedsToUpdateReticleState)
   > SetPointOfAttachmentState!(execute, self, ReticleState:GrappleKnifeReticleState Enum=@n1.ReticleState, NeedsToResetReticleState:bool=@n1.NeedsToResetReticleState) -> (then) {is=title K2Node_CallFunction}
   n4 = SetRenderTransform!(execute=@n2.else, self, InTransform:Widget Transform Structure) -> (then)
  ]}
 WorldToScreen {nodes [
   n1 = WorldToScreen!() -> (then, WorldLocation:Vector=~) {is=title K2Node_FunctionEntry}
   > SetWidgetCZonesAutomation!(execute, bEnabled:bool=true) -> (then)
   > Set TmpPos(TmpPos:Vector 2D Structure=ProjectToWidgetPosition!(PlayerController:Player Controller Object Reference=GetOwningPlayer(), WorldLocation:Vector=@n1, bPlayerViewportRelative:bool=false) -> (ScreenPosition:Vector 2D Structure, ReturnValue:bool)) -> (Output_Get:Vector 2D Structure)
   > SetWidgetCZonesAutomation(bEnabled=false)
   n2 = BreakVector2D(InVec=GetViewportSize!() -> (ReturnValue:Vector 2D Structure))
   n3 = MakeVector2D(X=float / float(A=@n2, B=2), Y=float / float(A=@n2.Y, B=2))
   n4 = Get TmpPos() -> (TmpPos:Vector 2D Structure)
   > Return Node!(
    execute
    ScreenLocation:Vector 2D Structure=vector2d / vector2d(
     A:Vector 2D Structure=vector2d - vector2d(A:Vector 2D Structure=@n4, B:Vector 2D Structure=@n3) -> (ReturnValue:Vector 2D Structure) {is=title K2Node_PromotableOperator}
     B:float=GetViewportScale!() -> (ReturnValue:float)
    ) -> (ReturnValue:Vector 2D Structure) {is=title K2Node_PromotableOperator}
   )
  ]}
]
variables [
 bLatestTargetingLedge {type=boolean}
 bTargetingLedge {type=boolean}
 bUpdateReticleLocation {type=boolean}
 LatestWorldLocation {type=Vector}
 PreviousReticleState {type=GrappleKnifeReticleState, default=NoTarget}
 TargetingLedge {type=boolean}
]
widgets {
 info {inheritedWidgetCount=0, parentClass=&UserWidget, rootWidgetClass=&Overlay, widgetCount=2}
 widgets [
  Overlay_LockedOn {widgetClassPath=&Overlay, bIsVariable=true, widget=Overlay_LockedOn}
  Image_ReticlePointOfAttachment {
   widgetClassPath=&Image
   bIsVariable=true
   parent=Overlay_LockedOn
   widget=Image_ReticlePointOfAttachment
   properties {
    brush {
     bIsDynamicallyLoaded=false
     drawAs=Image
     imageType=NoImage
     mirroring=NoMirror
     resourceName=None
     resourceObject=/GrappleWeapon/UI/Reticle/Material/M_UI_GrappleWeapon_PointOfAttachment
     tiling=NoTile
     imageSize=<45 45>
     margin=<ltrb 0 0 0 0>
     outlineSettings {
      bUseBrushTransparency=false
      roundingType=HalfHeightRadius
      width=0
      color {colorUseRule=UseColor_Specified, specifiedColor=<rgba 0 0 0 0>}
      cornerRadii=<0 0 0 0>
     }
     tintColor {colorUseRule=UseColor_Specified, specifiedColor=<rgba 1 1 1 1>}
     uVRegion {bIsValid=false, max=<0 0>, min=<0 0>}
    }
    renderTransform {angle=0, scale=<2 2>, shear=<0 0>, translation=<0 0>}
   }
   slot {class=OverlaySlot, properties {horizontalAlignment=HAlign_Fill, verticalAlignment=VAlign_Fill}}
  }
 ]
}
