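# Grapple Knife weapon reticle widget showing targeting state, attachment point and charges.
# It binds to the weapon's reticle-state, ammo and no-ammo-fire events, playing material-parameter animations for lock-on, zip, attack and out-of-ammo.
# A dynamic material draws charges, with cooldown fill updated by timer below full ammo and spend rotation lerped through an animated setter variable.
# Skydiving or parachuting hides charges and attachment marker; DBNO hides the central reticle.
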WidgetBlueprint /GrappleWeapon/UI/Reticle/WBP_Reticle_GrappleKnife
animations [
 Anim_Intro {
  binding_count=1
  length_seconds=0.3
  object=&self:Anim_Intro
  sequence {bindings [Reticle {
     id=#1
     possessed_class=Overlay
     tracks [
      {
       type=MovieSceneFloatTrack
       display_name=Render Opacity
       sections [{type=MovieSceneFloatSection, channels [None {type=MovieSceneScriptingFloatChannel, keys [[0,0],[6,1]]}]}]
      }
      {
       type=MovieScene2DTransformTrack
       sections [{
         type=MovieScene2DTransformSection
         channels [
          Translation.X {type=MovieSceneScriptingFloatChannel}
          Translation.Y {type=MovieSceneScriptingFloatChannel}
          Angle {type=MovieSceneScriptingFloatChannel}
          Scale.X {type=MovieSceneScriptingFloatChannel, keys [[0,1.25],[6,1]]}
          Scale.Y {type=MovieSceneScriptingFloatChannel, keys [[0,1.25],[6,1]]}
          Shear.X {type=MovieSceneScriptingFloatChannel}
          Shear.Y {type=MovieSceneScriptingFloatChannel}
         ]
        }]
      }
     ]
    }]}
 }
 Anim_LockedOn {
  binding_count=1
  length_seconds=0.4
  object=&self:Anim_LockedOn
  sequence {bindings [Image_ReticleVisual {
     id=#2
     possessed_class=Image
     tracks [{
       type=MovieSceneWidgetMaterialTrack
       sections [{
         type=MovieSceneParameterSection
         end_frame=8
         start_frame=0
         channels [
          BracketsScale {type=MovieSceneScriptingFloatChannel, keys [[0,1],[5,0.8]]}
          BracketsOpacity {type=MovieSceneScriptingFloatChannel, keys [[0,0.25],[8,0.5]]}
         ]
        }]
      }]
    }]}
 }
 Anim_NoTarget {
  binding_count=1
  length_seconds=0.3
  object=&self:Anim_NoTarget
  sequence {bindings [Image_ReticleVisual {
     id=#3
     possessed_class=Image
     tracks [{
       type=MovieSceneWidgetMaterialTrack
       sections [{
         type=MovieSceneParameterSection
         end_frame=6
         start_frame=0
         channels [
          BracketsScale {type=MovieSceneScriptingFloatChannel, const [1,1]}
          BracketsOpacity {type=MovieSceneScriptingFloatChannel, const [0.25,1]}
          ChangetoAttack {type=MovieSceneScriptingFloatChannel, const [0,1]}
          TrapezoidOpacity {type=MovieSceneScriptingFloatChannel, const [0,1]}
          TrapezoidGlowOpacity {type=MovieSceneScriptingFloatChannel, const [0,1]}
         ]
        }]
      }]
    }]}
 }
 Anim_LockedOn_Loop {
  binding_count=1
  length_seconds=0.8759
  object=&self:Anim_LockedOn_Loop
  sequence {bindings [Image_ReticleVisual {
     id=#4
     possessed_class=Image
     tracks [{
       type=MovieSceneWidgetMaterialTrack
       sections [{
         type=MovieSceneParameterSection
         end_frame=17
         start_frame=0
         channels [
          BracketsScale {type=MovieSceneScriptingFloatChannel, const [0.8,1]}
          AttachmentInSight {type=MovieSceneScriptingFloatChannel, const [1,1]}
          BracketsOpacity {type=MovieSceneScriptingFloatChannel, const [0.5,1]}
         ]
        }]
      }]
    }]}
 }
 Anim_LockedOn_WithoutScale {
  binding_count=1
  length_seconds=0.4
  object=&self:Anim_LockedOn_WithoutScale
  sequence {bindings [Image_ReticleVisual {
     id=#5
     possessed_class=Image
     tracks [{
       type=MovieSceneWidgetMaterialTrack
       sections [{
         type=MovieSceneParameterSection
         end_frame=8
         start_frame=0
         channels [BracketsOpacity {type=MovieSceneScriptingFloatChannel, const [0.5,1]}]
        }]
      }]
    }]}
 }
 Anim_SpendCharge {
  binding_count=2
  length_seconds=0.6
  object=&self:Anim_SpendCharge
  sequence {bindings [
    Image_Charges {
     id=#6
     possessed_class=Image
     tracks [{
       type=MovieSceneWidgetMaterialTrack
       sections [{
         type=MovieSceneParameterSection
         end_frame=12
         start_frame=0
         channels [
          Final Charge Highlight Anim {type=MovieSceneScriptingFloatChannel, keys [[0,0],[1.6683,1],[8,0]]}
          Final Charge Opacity {type=MovieSceneScriptingFloatChannel, keys [[1.6683,1],[4.1703,0],[8,1]]}
         ]
        }]
      }]
    }
    WBP_Reticle_GrappleKnife_C_144 {
     id=#7
     possessed_class=WBP_Reticle_GrappleKnife_C
     tracks [{
       type=MovieSceneDoubleTrack
       display_name=Charge Rotation Animator
       sections [{type=MovieSceneDoubleSection, end_frame=11, start_frame=0, channels [None {type=MovieSceneScriptingDoubleChannel, keys [[8,0],[11.9993,1]]}]}]
      }]
    }
   ]}
 }
 Anim_OnChargeGained {
  binding_count=1
  length_seconds=0.25
  object=&self:Anim_OnChargeGained
  sequence {bindings [Image_Charges {
     id=#8
     possessed_class=Image
     tracks [{
       type=MovieSceneWidgetMaterialTrack
       sections [{
         type=MovieSceneParameterSection
         end_frame=5
         start_frame=0
         channels [
          First Charge Highlight Anim {type=MovieSceneScriptingFloatChannel, keys [[0,0],[2,1],[4.982,0]]}
          First Charge Glow Anim {type=MovieSceneScriptingFloatChannel, keys [[0,0],[2,1],[5,0]]}
          Glow Width {type=MovieSceneScriptingFloatChannel, keys [[0,0.9],[2,0],[5,0.9]]}
         ]
        }]
      }]
    }]}
 }
 Anim_StartZipping {
  binding_count=1
  length_seconds=0.5
  object=&self:Anim_StartZipping
  sequence {bindings [Image_ReticleVisual {
     id=#9
     possessed_class=Image
     tracks [{
       type=MovieSceneWidgetMaterialTrack
       sections [{
         type=MovieSceneParameterSection
         end_frame=10
         start_frame=0
         channels [
          BracketsScale {type=MovieSceneScriptingFloatChannel, const [0.8,1]}
          ExagonScale {type=MovieSceneScriptingFloatChannel, const [0.32,1]}
          CenterOpacity {type=MovieSceneScriptingFloatChannel, keys [[0,1],[4,0],[5,0],[10,1]]}
          DotScale {type=MovieSceneScriptingFloatChannel, const [-0.02,1]}
          AttachmentInSight {type=MovieSceneScriptingFloatChannel, const [1,1]}
          ChangetoZip {type=MovieSceneScriptingFloatChannel, keys [[0,0],[4,0],[5,1]]}
          BracketsOpacity {type=MovieSceneScriptingFloatChannel, const [0.5,1]}
         ]
        }]
      }]
    }]}
 }
 Anim_AttackStart {
  binding_count=1
  length_seconds=0.3371
  object=&self:Anim_AttackStart
  sequence {bindings [Image_ReticleVisual {
     id=#10
     possessed_class=Image
     tracks [{
       type=MovieSceneWidgetMaterialTrack
       sections [{
         type=MovieSceneParameterSection
         end_frame=6
         start_frame=0
         channels [
          CenterOpacity {type=MovieSceneScriptingFloatChannel, keys [[0,1],[2.4837,0],[3.6913,1]]}
          ChangetoAttack {type=MovieSceneScriptingFloatChannel, keys [[0,0],[2.4837,0],[2.774,1]]}
          TrapezoidOpacity {type=MovieSceneScriptingFloatChannel, keys [[0,0],[1.7513,0],[4.6123,0.9984],[4.6797,0.7]]}
          TrapezoidGlowOpacity {type=MovieSceneScriptingFloatChannel, keys [[0,0],[2.619,0],[4.235,0.5],[6.741,0]]}
         ]
        }]
      }]
    }]}
 }
 Anim_AttackFinish {
  binding_count=1
  length_seconds=5
  object=&self:Anim_AttackFinish
  sequence {bindings [Image_ReticleVisual {
     id=#11
     possessed_class=Image
     tracks [{
       type=MovieSceneWidgetMaterialTrack
       sections [{
         type=MovieSceneParameterSection
         end_frame=5
         start_frame=0
         channels [
          CenterOpacity {type=MovieSceneScriptingFloatChannel, const [1,1]}
          ChangetoAttack {type=MovieSceneScriptingFloatChannel, const [0,1]}
          TrapezoidOpacity {type=MovieSceneScriptingFloatChannel, const [0,1]}
          TrapezoidGlowOpacity {type=MovieSceneScriptingFloatChannel, const [0,1]}
         ]
        }]
      }]
    }]}
 }
 Anim_OutOfAmmo {
  binding_count=1
  length_seconds=0.9999
  object=&self:Anim_OutOfAmmo
  sequence {bindings [Image_Charges {
     id=#12
     possessed_class=Image
     tracks [
      {
       type=MovieSceneFloatTrack
       display_name=Render Opacity
       sections [{
         type=MovieSceneFloatSection
         end_frame=19
         start_frame=0
         channels [None {type=MovieSceneScriptingFloatChannel, keys [[0,1],[4,0.5],[7,1],[10,0.5],[13,1],[16,0.5],[19.997,1]]}]
        }]
      }
      {
       type=MovieScene2DTransformTrack
       sections [{
         type=MovieScene2DTransformSection
         end_frame=19
         start_frame=0
         channels [
          Translation.X {type=MovieSceneScriptingFloatChannel, keys [[0,94],[4,96],[7,93],[10,96],[13,93],[16,96],[19.997,94]]}
          Translation.Y {type=MovieSceneScriptingFloatChannel, const [20,1]}
         ]
        }]
      }
     ]
    }]}
 }
]
graphs [
 EventGraph {nodes [
   NoAmmoFire() {is=title K2Node_CustomEvent}
   n1 = Get Anim_OutOfAmmo() -> (Anim_OutOfAmmo:Widget Animation Object Reference)
   > n2 = Branch(Condition=IsAnimationPlaying(InAnimation=@n1))
   PlayAnimation(execute=@n2.else, InAnimation=@n1)
   Event OnInitialized()
   > n3 = GetDynamicMaterial(self=Get Image_Charges() -> (Image_Charges:Image Object Reference))
   > n4 = Set ChargesDynamicMaterial(ChargesDynamicMaterial:Material Instance Dynamic Object Reference=@n3) -> (Output_Get:Material Instance Dynamic Object Reference)
   n5 = Get Image_ReticleVisual() -> (Image_ReticleVisual:Image Object Reference)
   n6 = Get!(self:Image Object Reference=@n5) -> (
    Brush_TintColor:Slate Color Structure
    Brush_DrawAs:ESlateBrushDrawType Enum
    Brush_Tiling:ESlateBrushTileType Enum
    Brush_Mirroring:ESlateBrushMirrorType Enum
    Brush_ImageSize:Vector 2D Structure
    Brush_Margin:Margin Structure
    Brush_ResourceObject:Object Reference
    Brush_OutlineSettings:Slate Brush Outline Settings Structure
   )
   SetBrush(
    execute=[@n15.then, @n4.then]
    self=@n5
    InBrush=Make Slate Brush!(
     TintColor:Slate Color Structure=@n6
     DrawAs:ESlateBrushDrawType Enum=@n6.Brush_DrawAs
     Tiling:ESlateBrushTileType Enum=@n6.Brush_Tiling
     Mirroring:ESlateBrushMirrorType Enum=@n6.Brush_Mirroring
     ImageSize:Vector 2D Structure=Get ReticleSize() -> (ReticleSize:Vector 2D Structure)
     Margin:Margin Structure=@n6.Brush_Margin
     ResourceObject:Object Reference=Get ReticleMaterial() -> (ReticleMaterial:Material Instance Object Reference)
     OutlineSettings:Slate Brush Outline Settings Structure=@n6.Brush_OutlineSettings
    ) -> (SlateBrush:Slate Brush Structure) {is=title K2Node_MakeStruct}
   )
   > UpdateCooldownQueryTimer(self:Fort Weapon Reticle Extension Widget Base Object Reference)
   Event Construct()
   > Parent: Construct()
   > PlayAnimation(InAnimation=Get Anim_Intro() -> (Anim_Intro:Widget Animation Object Reference))
   > Set LatestUIState(LatestUIState:GrappleKnifeReticleState Enum=NewEnumerator2) -> (Output_Get:GrappleKnifeReticleState Enum=NewEnumerator0)
   > n7 = Cast To B_GrappleKnife!(
    execute
    Object=GetEquippedWeapon!(self:Fort Weapon Reticle Extension Widget Base Object Reference) -> (ReturnValue:Fort Weapon Object Reference)
   ) -> (then, CastFailed, AsB Grapple Knife:B Grapple Knife Object Reference) {is=title K2Node_DynamicCast}
   > n8 = Set Grapple Weapon Ref(Grapple Weapon Ref:B Grapple Knife Object Reference=@n7) -> (Output_Get:B Grapple Knife Object Reference)
   > Bind Event to On Grapple Knife Reticle State Changed(self:B Grapple Knife Object Reference=@n8, Delegate=Create Event()) {is=title K2Node_AddDelegate}
   > Bind Event to On Local Ammo Changed(self:Fort Weapon Object Reference=@n8, Delegate=Create Event()) {is=title K2Node_AddDelegate}
   > Bind Event to On No Ammo Fire(
    self:B Grapple Knife Object Reference=Get Grapple Weapon Ref() -> (Grapple Weapon Ref:B Grapple Knife Object Reference)
    Delegate=Create Event()
   ) {is=title K2Node_AddDelegate}
   n9 = Get Grapple Weapon Ref()
   > InitializeCharges!(
    execute
    self
    TotalCharges:int=GetBulletsPerClip!(self:Fort Weapon Object Reference=@n9) -> (ReturnValue:int=0)
    CurrentCharges:int=GetLocalMagazineAmmoCount!(self:Fort Weapon Object Reference=@n9) -> (ReturnValue:int=0)
   ) -> (then)
   > Set NeedsToResetReticleState(NeedsToResetReticleState:bool=true) -> (Output_Get:bool)
   K2Node_Event_2 = OnLocalAmmoChanged!() -> (OutputDelegate, then, LocalCount:int, LocalRemaining:int) {is=title K2Node_CustomEvent}
   > UpdateCooldownQueryTimer()
   > UpdateChargeCount!(execute, self, NewChargeCount:int=@K2Node_Event_2.LocalCount) -> (then)
   > Branch(Condition="Equal (Integer)"(A=@K2Node_Event_2.LocalCount))
   > PlayAnimation(InAnimation=Get Anim_NoTarget() -> (Anim_NoTarget:Widget Animation Object Reference))
   > UpdatePointOfAttachment!(
    execute
    self:WBP Reticle Grapple Weapon Point Of Attachment Object Reference=Get WBP_Reticle_GrappleWeapon_PointOfAttachment() -> (WBP_Reticle_GrappleWeapon_PointOfAttachment:WBP Reticle Grapple Weapon Point Of Attachment Object Reference)
    NeedsToUpdatePosition:bool=true
    WorldLocation:Vector="0, 0, 0"
    TargetingLedge:bool=false
    NeedsToUpdateReticleState:bool=true
    ReticleState:GrappleKnifeReticleState Enum=NewEnumerator0
    NeedsToResetReticleState:bool=true
   ) -> (then)
   > Set NeedsToResetReticleState(NeedsToResetReticleState=true)
   > Set LatestUIState(LatestUIState=NewEnumerator0) -> (Output_Get=NewEnumerator0)
   Event UpdateRemainingCooldown() {is=title K2Node_Event}
   > n10 = GetRemainingCooldownTimer!(execute, self) -> (then, RemainingTime:double, TotalDuration:double)
   > n11 = Branch(Condition=OR Boolean(A="Equal (Float)"(A=@n10), B="Equal (Float)"(B=@n10.TotalDuration)))
   > SetScalarParameterValue(
    self=Get ChargesDynamicMaterial() -> (ChargesDynamicMaterial:Material Instance Dynamic Object Reference)
    ParameterName=Get FillProgressParam() -> (FillProgressParam:Name=None)
    Value=1
   )
   SetScalarParameterValue(
    execute=@n11.else
    self=Get ChargesDynamicMaterial()
    ParameterName=Get FillProgressParam() -> (FillProgressParam=None)
    Value=float - float(A=1, B=float / float(A=@n10, B=@n10.TotalDuration))
   )
   Event UpdateCooldownQueryTimer() {is=title K2Node_Event}
   > n12 = Get!(execute:Exec) -> (then:Exec, else:Exec, Grapple Weapon Ref:B Grapple Knife Object Reference)
   > n13 = Branch(Condition="integer >= integer"(A=GetLocalMagazineAmmoCount(self=@n12) -> (ReturnValue=0), B=GetBulletsPerClip(self=@n12) -> (ReturnValue=0)))
   > HandleCooldownTimer!(execute, self:Fort Weapon Reticle Extension Widget Base Object Reference, bActivateTimer:bool=false, TimerInterval:float=0) -> (then)
   HandleCooldownTimer(execute=@n13.else, bActivateTimer=true, TimerInterval=0.1)
   n14 = Event PreConstruct()
   > Parent: PreConstruct(IsDesignTime=@n14.IsDesignTime)
   > n15 = Branch(Condition=@n14.IsDesignTime)
   Event Destruct()
   > Parent: Destruct()
   n16 = Get Grapple Weapon Ref()
   > Is Valid(InputObject=@n16)
   > Unbind Event from On Grapple Knife Reticle State Changed(self:B Grapple Knife Object Reference=@n16, Delegate=Create Event()) {is=title K2Node_RemoveDelegate}
   n17 = Reroute Node(InputPin:Fort Weapon Object Reference) -> (OutputPin:Fort Weapon Object Reference)
   > Unbind Event from On Local Ammo Changed(self:Fort Weapon Object Reference=[@n16, @n17], Delegate=Create Event()) {is=title K2Node_RemoveDelegate}
   > Unbind Event from On No Ammo Fire(self:B Grapple Knife Object Reference=@n16, Delegate=Create Event()) {is=title K2Node_RemoveDelegate}
   Event OnAmmoClipAttributeChanged!() -> (OutputDelegate, then, AmmoClipValue:float) {is=title K2Node_Event}
   > Is Valid(InputObject=Get Grapple Weapon Ref())
   n18 = Get Grapple Weapon Ref()
   > InitializeCharges(TotalCharges=GetBulletsPerClip(self=@n18) -> (ReturnValue=0), CurrentCharges=GetLocalMagazineAmmoCount(self=@n18) -> (ReturnValue=0))
   Reroute Node(InputPin=@n17)
  ]}
 GetDesiredChargesRotation {nodes [
   GetDesiredChargesRotation() {is=title K2Node_FunctionEntry}
   > n1 = Get Scalar Parameter Value!(execute, self:Material Instance Dynamic Object Reference=Get ChargesDynamicMaterial(), ParameterName:Name=Segments) -> (then, ReturnValue:float) {is=title K2Node_CallFunction}
   > n2 = Branch(Condition="Equal (Float)"(A=@n1))
   > Return Node!(execute, DesiredRotation:double=0)
   n3 = Get TotalChargeCount() -> (TotalChargeCount:int=0)
   n4 = "Clamp (Integer)"(Value=int - int(A=@n3, B=Get CurrentChargeCount() -> (CurrentChargeCount:int=0)), Min=0, Max=@n3) -> (ReturnValue=0)
   Return Node(
    execute=@n2.else
    DesiredRotation=float * float(A=float / float(A=1, B=@n1), B="To Float (Integer)"(InInt=@n4)) {is=title K2Node_PromotableOperator}
   )
  ]}
 GetRemainingCooldownTimer {nodes [
   GetRemainingCooldownTimer() {is=title K2Node_FunctionEntry}
   > n1 = Cast To FortPlayerController(Object=GetOwningPlayer())
   > n2 = Get()
   n3 = GetInventoryItem!(self:Fort Weapon Object Reference=@n2) -> (ReturnValue:Fort Item Object Reference)
   > n4 = Is Valid(InputObject=@n3)
   n5 = GetCoolDownTimeRemaining!(self:Fort Item Object Reference=@n3, FortPC:Fort Player Controller Object Reference=@n1) -> (CooldownDuration:float, ReturnValue:float)
   > Return Node!(execute, RemainingTime:double=@n5.ReturnValue, TotalDuration:double=@n5)
   Return Node(execute=[@n1.CastFailed, @n2.else, @n4.Is Not Valid], RemainingTime=0, TotalDuration=0)
  ]}
 InitializeCharges {nodes [
   n1 = InitializeCharges!() -> (then, TotalCharges:int, CurrentCharges:int) {is=title K2Node_FunctionEntry}
   > n2 = Set TotalChargeCount(TotalChargeCount:int=@n1) -> (Output_Get:int=0)
   > n3 = Set CurrentChargeCount(CurrentChargeCount:int=@n1.CurrentCharges) -> (Output_Get:int=0)
   n4 = Get ChargesDynamicMaterial()
   > SetScalarParameterValue(self=@n4, ParameterName=Visible Segment Count, Value="To Float (Integer)"(InInt=@n2))
   > SetScalarParameterValue(self=@n4, ParameterName=Get FinalChargeHighlightAnimParam() -> (FinalChargeHighlightAnimParam:Name=None), Value=0)
   > SetScalarParameterValue(self=@n4, ParameterName=Get FirstChargeHighlightAnimParam() -> (FirstChargeHighlightAnimParam:Name=None), Value=0)
   > SetScalarParameterValue(self=@n4, ParameterName=Get ChargesAvailableParam() -> (ChargesAvailableParam:Name=None), Value="To Float (Integer)"(InInt=@n3))
   > SetScalarParameterValue(
    self=@n4
    ParameterName=Get ChargeRotationParam() -> (ChargeRotationParam:Name=None)
    Value=GetDesiredChargesRotation!(self) -> (DesiredRotation:double) {is=title K2Node_CallFunction}
   )
   > SetScalarParameterValue(self=@n4, ParameterName=Get FillProgressParam() -> (FillProgressParam=None), Value=0)
   > SetScalarParameterValue(self=Get ChargesDynamicMaterial(), ParameterName=Get FinalChargeOpacityParam() -> (FinalChargeOpacityParam:Name=None), Value=1)
  ]}
 SetChargeRotationAnimator {nodes [
   n1 = SetChargeRotationAnimator!() -> (then, NewValue:double=~) {is=title K2Node_FunctionEntry}
   > SetScalarParameterValue(
    self=Get ChargesDynamicMaterial()
    ParameterName=Get ChargeRotationParam() -> (ChargeRotationParam=None)
    Value=Lerp(A=Get StartingChargeRotation() -> (StartingChargeRotation:double), B=GetDesiredChargesRotation(), Alpha=@n1)
   )
  ]}
 UpdateChargeCount {nodes [
   n1 = UpdateChargeCount!() -> (then, NewChargeCount:int) {is=title K2Node_FunctionEntry}
   > Branch(Condition="Not Equal (Integer)"(A=@n1, B=Get CurrentChargeCount() -> (CurrentChargeCount=0)))
   > n2 = Branch(Condition="integer > integer"(A=Get NewChargeCount() -> (NewChargeCount:int=0), B=Get CurrentChargeCount() -> (CurrentChargeCount=0)))
   > Set CurrentChargeCount(CurrentChargeCount=Get NewChargeCount() -> (NewChargeCount=0)) -> (Output_Get=0)
   > SetScalarParameterValue(
    self=Get ChargesDynamicMaterial()
    ParameterName=Get ChargesAvailableParam() -> (ChargesAvailableParam=None)
    Value="To Float (Integer)"(InInt=Get NewChargeCount() -> (NewChargeCount=0))
   )
   > SetScalarParameterValue(self=Get ChargesDynamicMaterial(), ParameterName=Get ChargeRotationParam() -> (ChargeRotationParam=None), Value=GetDesiredChargesRotation())
   > SetScalarParameterValue(self=Get ChargesDynamicMaterial(), ParameterName=Get ChargeRotationParam() -> (ChargeRotationParam=None), Value=GetDesiredChargesRotation())
   > SetScalarParameterValue(self=Get ChargesDynamicMaterial(), ParameterName=Get FinalChargeOpacityParam() -> (FinalChargeOpacityParam=None), Value=1)
   > SetScalarParameterValue(self=Get ChargesDynamicMaterial(), ParameterName=Get FinalChargeHighlightAnimParam() -> (FinalChargeHighlightAnimParam=None), Value=0)
   > SetScalarParameterValue(self=Get ChargesDynamicMaterial(), ParameterName=Get FillProgressParam() -> (FillProgressParam=None), Value=0)
   > PlayAnimation(InAnimation=Get Anim_OnChargeGained() -> (Anim_OnChargeGained:Widget Animation Object Reference))
   Set CurrentChargeCount(execute=@n2.else, CurrentChargeCount=Get NewChargeCount() -> (NewChargeCount=0)) -> (Output_Get=0)
   n3 = Get NewChargeCount() -> (NewChargeCount=0)
   > SetScalarParameterValue(
    self=Get ChargesDynamicMaterial()
    ParameterName=Get ChargesAvailableParam() -> (ChargesAvailableParam=None)
    Value="To Float (Integer)"(InInt=int + int(A=@n3))
   )
   > n4 = Get Scalar Parameter Value(self=Get ChargesDynamicMaterial(), ParameterName=Get ChargeRotationParam() -> (ChargeRotationParam=None))
   > Set StartingChargeRotation(StartingChargeRotation:double=@n4) -> (Output_Get:double)
   > PlayAnimation(InAnimation=Get Anim_SpendCharge() -> (Anim_SpendCharge:Widget Animation Object Reference))
  ]}
 UpdateGrappleWeaponReticleAnimation {nodes [
   n1 = UpdateGrappleWeaponReticleAnimation!() -> (then, ReticleState:GrappleKnifeReticleState Enum=NewEnumerator0) {is=title K2Node_FunctionEntry}
   n2 = Get Grapple Weapon Ref()
   > n3 = Branch(Condition="integer > integer"(A=GetRemainingAmmo!(self:Fort Weapon Object Reference=@n2) -> (ReturnValue:int=0)))
   n4 = Switch on GrappleKnifeReticleState!(execute=[@n8.then, @n3.then], Selection:GrappleKnifeReticleState Enum=@n1) -> (
    NewEnumerator0:Exec
    NewEnumerator1:Exec
    NewEnumerator7:Exec
    NewEnumerator2:Exec
    NewEnumerator3:Exec
    NewEnumerator4:Exec
    NewEnumerator5:Exec
    NewEnumerator6:Exec
   ) {is=title K2Node_SwitchEnum}
   > PlayAnimation(InAnimation=Get Anim_NoTarget())
   n5 = "Equal (Enum)"(
    A:GrappleKnifeReticleState Enum=Get LatestUIState() -> (LatestUIState:GrappleKnifeReticleState Enum=NewEnumerator0)
    B:GrappleKnifeReticleState Enum=NewEnumerator0
   )
   n6 = Get NeedsToResetReticleState() -> (NeedsToResetReticleState:bool)
   PlayAnimation(
    execute=[@n4.NewEnumerator1, @n4.NewEnumerator7]
    InAnimation=Select(
     Option 0:Widget Animation Object Reference=Get Anim_LockedOn_WithoutScale() -> (Anim_LockedOn_WithoutScale:Widget Animation Object Reference)
     Option 1:Widget Animation Object Reference=Get Anim_LockedOn() -> (Anim_LockedOn:Widget Animation Object Reference)
     Index=OR Boolean(A=@n5, B=@n6)
    ) -> (ReturnValue:Widget Animation Object Reference)
   )
   > Set NeedsToResetReticleState(NeedsToResetReticleState=false)
   n7 = Get Grapple Weapon Ref()
   Branch(execute=@n4.NewEnumerator3, Condition="integer > integer"(A=GetRemainingAmmo(self=@n7) -> (ReturnValue=0)))
   > PlayAnimation(InAnimation=Get Anim_StartZipping() -> (Anim_StartZipping:Widget Animation Object Reference))
   PlayAnimation(execute=@n4.NewEnumerator5, InAnimation=Get Anim_AttackStart() -> (Anim_AttackStart:Widget Animation Object Reference))
   PlayAnimation(execute=@n4.NewEnumerator6, InAnimation=Get Anim_AttackFinish() -> (Anim_AttackFinish:Widget Animation Object Reference))
   n8 = Branch(
    execute=@n3.else
    Condition=AND Boolean(
     A="Not Equal (Enum)"!(A:GrappleKnifeReticleState Enum=@n1, B:GrappleKnifeReticleState Enum=NewEnumerator1) -> (ReturnValue:bool)
     B="Not Equal (Enum)"(A=@n1, B=NewEnumerator0)
    )
   )
  ]}
 UpdateGrappleWeaponState {nodes [
   n1 = UpdateGrappleWeaponState!() -> (then, UIState:GrappleKnifeReticleState Enum=NewEnumerator0, ZipDistance:double=~, WorldLocation:Vector=~, bTargetingLedge:bool=~) {is=title K2Node_FunctionEntry}
   n2 = Get LatestUIState() -> (LatestUIState=NewEnumerator0)
   n3 = GetRemainingAmmo(self=Get Grapple Weapon Ref()) -> (ReturnValue=0)
   > n4 = Branch(Condition=AND Boolean(A="Not Equal (Enum)"(A=@n1, B=@n2), B="integer > integer"(A=@n3)))
   > UpdateGrappleWeaponReticleAnimation!(execute, self, ReticleState:GrappleKnifeReticleState Enum=@n1) -> (then) {is=title K2Node_CallFunction}
   > n5 = Set LatestUIState(LatestUIState=@n1) -> (Output_Get=NewEnumerator0)
   > n6 = UpdatePointOfAttachment(
    self=Get WBP_Reticle_GrappleWeapon_PointOfAttachment()
    NeedsToUpdatePosition=true
    WorldLocation=@n1.WorldLocation
    TargetingLedge=@n1.bTargetingLedge
    NeedsToUpdateReticleState=true
    ReticleState=@n5
    NeedsToResetReticleState=Get NeedsToResetReticleState()
   )
   UpdateReticleVisibility(execute=[@n7.then, @n6.then])
   n7 = UpdatePointOfAttachment(
    execute=@n4.else
    self=Get WBP_Reticle_GrappleWeapon_PointOfAttachment()
    NeedsToUpdatePosition=true
    WorldLocation=@n1.WorldLocation
    TargetingLedge=@n1.bTargetingLedge
    NeedsToUpdateReticleState=false
    ReticleState=Get LatestUIState() -> (LatestUIState=NewEnumerator0)
    NeedsToResetReticleState=false
   )
  ]}
 UpdateReticleVisibility {nodes [
   UpdateReticleVisibility() {is=title K2Node_FunctionEntry}
   > n1 = Cast To FortPlayerController(Object=GetOwningPlayer())
   > n2 = Sequence()
   n3 = GetPlayerPawn!(self:Fort Player Controller Object Reference=@n1) -> (ReturnValue:Fort Player Pawn Object Reference)
   n4 = IsParachuteOpen!(self:Fort Player Pawn Object Reference=@n3) -> (ReturnValue:bool)
   n5 = IsSkydiving!(self:Fort Player Pawn Object Reference=@n3) -> (ReturnValue:bool)
   > SetVisibility(
    self=[Get WBP_Reticle_GrappleWeapon_PointOfAttachment(), Get Image_Charges()]
    InVisibility=Select(
     Option 0:ESlateVisibility Enum=HitTestInvisible
     Option 1:ESlateVisibility Enum=Hidden
     Index=OR Boolean!(A=@n5, B=@n4, C:bool) -> (ReturnValue)
    ) -> (ReturnValue:ESlateVisibility Enum)
   )
   SetVisibility(
    execute=@n2.then_1
    self=Get Image_ReticleVisual()
    InVisibility=Select(Option 0=HitTestInvisible, Option 1=Hidden, Index=IsDBNO!(self:Fort Pawn Object Reference=@n3) -> (ReturnValue:bool))
   )
  ]}
 WorldToScreen {nodes [
   n1 = WorldToScreen!() -> (then, WorldLocation:Vector=~) {is=title K2Node_FunctionEntry}
   > SetWidgetCZonesAutomation!(execute, bEnabled:bool=true) -> (then)
   > Set TmpPos(TmpPos:Vector 2D Structure=ProjectToWidgetPosition!(PlayerController:Player Controller Object Reference=GetOwningPlayer(), WorldLocation:Vector=@n1, bPlayerViewportRelative:bool=false) -> (ScreenPosition:Vector 2D Structure, ReturnValue:bool)) -> (Output_Get:Vector 2D Structure)
   > SetWidgetCZonesAutomation(bEnabled=false)
   n2 = BreakVector2D(InVec=GetViewportSize!() -> (ReturnValue:Vector 2D Structure))
   n3 = MakeVector2D(X=float / float(A=@n2, B=2), Y=float / float(A=@n2.Y, B=2))
   n4 = Get TmpPos() -> (TmpPos:Vector 2D Structure)
   > Return Node!(
    execute
    ScreenLocation:Vector 2D Structure=vector2d / vector2d(
     A:Vector 2D Structure=vector2d - vector2d(A:Vector 2D Structure=@n4, B:Vector 2D Structure=@n3) -> (ReturnValue:Vector 2D Structure) {is=title K2Node_PromotableOperator}
     B:float=GetViewportScale!() -> (ReturnValue:float)
    ) -> (ReturnValue:Vector 2D Structure) {is=title K2Node_PromotableOperator}
   )
  ]}
]
variables [
 bLatestTargetingLedge {type=boolean}
 bUpdateReticleColor {type=boolean}
 bUpdateReticleLocation {type=boolean, default=true}
 ChargeRotationAnimator {type=number}
 ChargeRotationParam {type=string, default=Charge Rotation}
 ChargesAvailableParam {type=string, default=Charges Available}
 ChargesDynamicMaterial {type=MaterialInstanceDynamic}
 CurrentChargeCount {type=integer, default=4}
 FillProgressParam {type=string, default=Fill Progress}
 FinalChargeHighlightAnimParam {type=string, default=Final Charge Highlight Anim}
 FinalChargeOpacityParam {type=string, default=Final Charge Opacity}
 FirstChargeHighlightAnimParam {type=string, default=First Charge Highlight Anim}
 Grapple Weapon Ref {type=B_GrappleKnife_C, default=null}
 LatestUIState {type=GrappleKnifeReticleState, default=NoTarget}
 LatestWorldLocation {type=Vector}
 NeedsToResetReticleState {type=boolean}
 ReticleMaterial {type=MaterialInstance, default=/GrappleWeapon/UI/Reticle/Material/MI_UI_GrappleWeaponReticle}
 ReticleSize {type=Vector2D, default=<90 90>}
 StartingChargeRotation {type=number}
 TargetingLedge {type=boolean}
 TotalChargeCount {type=integer, default=6}
]
widgets {
 info {
  inheritedWidgetCount=1
  parentClass=/Script/FortniteUI.FortWeaponReticleExtensionWidgetBase
  rootWidgetClass=&Overlay
  widgetCount=5
 }
 widgets [
  Reticle {widgetClassPath=&Overlay, widget=Reticle}
  Image_ReticleVisual {
   widgetClassPath=&Image
   bIsVariable=true
   parent=Reticle
   widget=Image_ReticleVisual
   properties {brush {
     bIsDynamicallyLoaded=false
     drawAs=Image
     imageType=NoImage
     mirroring=NoMirror
     resourceName=None
     resourceObject=&MI_UI_GrappleWeaponReticle
     tiling=NoTile
     imageSize=<90 90>
     margin=<ltrb 0 0 0 0>
     outlineSettings {
      bUseBrushTransparency=false
      roundingType=HalfHeightRadius
      width=0
      color {colorUseRule=UseColor_Specified, specifiedColor=<rgba 0 0 0 0>}
      cornerRadii=<0 0 0 0>
     }
     tintColor {colorUseRule=UseColor_Specified, specifiedColor=<rgba 1 1 1 1>}
     uVRegion {bIsValid=false, max=<0 0>, min=<0 0>}
    }}
   slot {class=OverlaySlot, properties {horizontalAlignment=HAlign_Center, verticalAlignment=VAlign_Center}}
  }
  WBP_Reticle_GrappleWeapon_PointOfAttachment {
   widgetClassPath=&/WBP_Reticle_GrappleWeapon_PointOfAttachment
   bIsVariable=true
   parent=Reticle
   widget=WBP_Reticle_GrappleWeapon_PointOfAttachment
   slot {class=OverlaySlot, properties {horizontalAlignment=HAlign_Center, verticalAlignment=VAlign_Center}}
  }
  Image_Charges {
   widgetClassPath=&Image
   bIsVariable=true
   parent=Reticle
   widget=Image_Charges
   properties {
    brush {
     bIsDynamicallyLoaded=false
     drawAs=Image
     imageType=NoImage
     mirroring=NoMirror
     resourceName=None
     resourceObject=/GrappleWeapon/UI/Reticle/Material/MI_UI_GrappleWeapon_Charges
     tiling=NoTile
     imageSize=<128 128>
     margin=<ltrb 0 0 0 0>
     outlineSettings {
      bUseBrushTransparency=false
      roundingType=HalfHeightRadius
      width=0
      color {colorUseRule=UseColor_Specified, specifiedColor=<rgba 0 0 0 0>}
      cornerRadii=<0 0 0 0>
     }
     tintColor {colorUseRule=UseColor_Specified, specifiedColor=<rgba 1 1 1 1>}
     uVRegion {bIsValid=false, max=<0 0>, min=<0 0>}
    }
    renderTransform {angle=0, scale=<1 1>, shear=<0 0>, translation=<94 20>}
   }
   slot {class=OverlaySlot, properties {horizontalAlignment=HAlign_Center, verticalAlignment=VAlign_Center}}
  }
  TutorialHighlight {widgetClassPath=&TutorialHighlight}
 ]
}
